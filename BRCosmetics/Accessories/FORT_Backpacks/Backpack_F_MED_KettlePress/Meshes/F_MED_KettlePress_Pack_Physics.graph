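# Physics asset for the KettlePress backpack cosmetic, defining bodies and joints for its dangling belts, handle, buckle and two zippers.
# Left and right belts are six-segment capsule chains, linked from a master body to the central pack sphere, with limited swing and twist between segments.
# Handle, buckle and zippers are single bodies jointed to their own masters.
# Both zippers lock twist, while the right belt's pack joint frees all rotation.
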PhysicsAsset /BRCosmetics/Accessories/FORT_Backpacks/Backpack_F_MED_KettlePress/Meshes/F_MED_KettlePress_Pack_Physics
bodies {
 dyn_belt_l_01 [{shapeType=Capsule, length=2.3778, radius=0.8967, shapeName=dyn_belt_l_01_capsule, center=<1.8372 0 0>, rotation=<rot 90 -19.4711 -19.4713>}]
 dyn_belt_l_02 [{shapeType=Capsule, length=2.3778, radius=0.8967, shapeName=dyn_belt_l_02_capsule, center=<1.8372 0 0>, rotation=<rot 90 -90 -90.0002>}]
 dyn_belt_l_03 [{shapeType=Capsule, length=2.3778, radius=0.8967, shapeName=dyn_belt_l_03_capsule, center=<1.8372 0 0>, rotation=<rot 90 -160.5288 -160.529>}]
 dyn_belt_l_04 [{shapeType=Capsule, length=2.3778, radius=0.8967, shapeName=dyn_belt_l_04_capsule, center=<1.8372 0 0>, rotation=<rot 90 -180 179.9999>}]
 dyn_belt_l_05 [{shapeType=Capsule, length=2.3778, radius=0.8967, shapeName=dyn_belt_l_05_capsule, center=<1.8372 0 0>, rotation=<rot 90 180 180>}]
 dyn_belt_l_06 [{shapeType=Capsule, length=0.5545, radius=0.8967, shapeName=dyn_belt_l_06_capsule, center=<0.0345 0 0>, rotation=<rot 90 0.0148 0.0148>}]
 dyn_belt_l_master [{
   shapeType=Capsule
   length=0.9818
   radius=5.6787
   shapeName=dyn_belt_l_master_capsule
   center=<-0.7413 0.4897 -6.5462>
   rotation=<rot -8.1382 -179.9999 90.0001>
  }]
 dyn_belt_r_01 [{shapeType=Capsule, length=4.4552, radius=0.6957, shapeName=dyn_belt_r_01_capsule, center=<2.2478 0 0>, rotation=<rot -90 -179.9604 179.9605>}]
 dyn_belt_r_02 [{shapeType=Capsule, length=4.4552, radius=0.6957, shapeName=dyn_belt_r_02_capsule, center=<2.2478 0 0>, rotation=<rot -90 -179.9209 179.9209>}]
 dyn_belt_r_03 [{shapeType=Capsule, length=4.4552, radius=0.6957, shapeName=dyn_belt_r_03_capsule, center=<2.2478 0 0>, rotation=<rot -90 -179.9994 179.9994>}]
 dyn_belt_r_04 [{shapeType=Capsule, length=4.4552, radius=0.6957, shapeName=dyn_belt_r_04_capsule, center=<2.2478 0 0>, rotation=<rot -90 0.0002 -0.0003>}]
 dyn_belt_r_05 [{shapeType=Capsule, length=3.1325, radius=0.6957, shapeName=dyn_belt_r_05_capsule, center=<1.7891 0 0>, rotation=<rot -90 0 -0.0001>}]
 dyn_belt_r_06 [{shapeType=Capsule, length=0.7451, radius=0.6957, shapeName=dyn_belt_r_06_capsule, center=<0.2897 0 0>, rotation=<rot -90 -179.9604 179.9605>}]
 dyn_belt_r_master [{shapeType=Sphere, radius=0.6026, shapeName=dyn_belt_r_master_sphere}]
 dyn_buckle_up_01 [{shapeType=Capsule, length=2.3977, radius=0.9141, shapeName=dyn_buckle_up_01_capsule, center=<1.4986 0 0.8832>, rotation=<rot -90 0 0>}]
 dyn_buckle_up_master [{
   shapeType=Capsule
   length=2.8368
   radius=11.607
   shapeName=dyn_buckle_up_master_capsule
   center=<0.5372 0 -11.814>
   rotation=<rot -0.0002 0 90>
  }]
 dyn_handle_01 [{shapeType=Capsule, length=0, radius=1.5111, shapeName=dyn_handle_01_capsule, center=<0.0171 0 0>, rotation=<rot -90 0 0>}]
 dyn_handle_master [{shapeType=Sphere, radius=0.6026, shapeName=dyn_handle_master_sphere}]
 dyn_zipper_01 [{shapeType=Capsule, length=2.7543, radius=0.6416, shapeName=dyn_zipper_01_capsule, center=<1.3853 0 0.1419>, rotation=<rot -90 -0.0002 0>}]
 dyn_zipper_master [{
   shapeType=Capsule
   length=5.9272
   radius=1.3469
   shapeName=dyn_zipper_master_capsule
   center=<2.3319 0 -1.9643>
   rotation=<rot 87.7726 0.0001 0.0003>
  }]
 dyn_zipper_upper_01 [{
   shapeType=Capsule
   length=2.792
   radius=0.8854
   shapeName=dyn_zipper_upper_01_capsule
   center=<1.5727 0 0.4233>
   rotation=<rot -90 180 179.9998>
  }]
 dyn_zipper_upper_master [{
   shapeType=Capsule
   length=3.0876
   radius=2.7204
   shapeName=dyn_zipper_upper_master_capsule
   center=<1.7832 0 -3.146>
   rotation=<rot -85 -179.9998 179.9993>
  }]
 pack [{shapeType=Sphere, radius=1.9839, shapeName=pack_sphere}]
}
constraints [
 {bone1Name=dyn_handle_01, bone2Name=dyn_handle_master, swing1LimitDegrees=5, swing2LimitDegrees=5, twistLimitDegrees=5}
 {bone1Name=dyn_zipper_01, bone2Name=dyn_zipper_master, swing1LimitDegrees=15, swing2LimitDegrees=45, twistLimitDegrees=45, twistMotion=Locked}
 {bone1Name=dyn_buckle_up_01, bone2Name=dyn_buckle_up_master, swing1LimitDegrees=3, swing2LimitDegrees=22, twistLimitDegrees=1}
 {bone1Name=dyn_belt_l_06, bone2Name=pack, swing1LimitDegrees=45, swing2LimitDegrees=45, twistLimitDegrees=45}
 {bone1Name=dyn_belt_l_06, bone2Name=dyn_belt_l_05, swing1LimitDegrees=10, swing2LimitDegrees=40, twistLimitDegrees=3}
 {bone1Name=dyn_belt_l_05, bone2Name=dyn_belt_l_04, swing1LimitDegrees=10, swing2LimitDegrees=40, twistLimitDegrees=3}
 {bone1Name=dyn_belt_l_04, bone2Name=dyn_belt_l_03, swing1LimitDegrees=10, swing2LimitDegrees=30, twistLimitDegrees=3}
 {bone1Name=dyn_belt_l_03, bone2Name=dyn_belt_l_02, swing1LimitDegrees=10, swing2LimitDegrees=40, twistLimitDegrees=3}
 {bone1Name=dyn_belt_l_02, bone2Name=dyn_belt_l_01, swing1LimitDegrees=10, swing2LimitDegrees=30, twistLimitDegrees=3}
 {bone1Name=dyn_belt_l_01, bone2Name=dyn_belt_l_master, swing1LimitDegrees=3, swing2LimitDegrees=30, twistLimitDegrees=3}
 {bone1Name=dyn_belt_r_06, bone2Name=dyn_belt_r_05, swing1LimitDegrees=10, swing2LimitDegrees=30, twistLimitDegrees=6}
 {bone1Name=dyn_belt_r_05, bone2Name=dyn_belt_r_04, swing1LimitDegrees=10, swing2LimitDegrees=40, twistLimitDegrees=6}
 {bone1Name=dyn_belt_r_04, bone2Name=dyn_belt_r_03, swing1LimitDegrees=10, swing2LimitDegrees=30, twistLimitDegrees=6}
 {bone1Name=dyn_belt_r_03, bone2Name=dyn_belt_r_02, swing1LimitDegrees=10, swing2LimitDegrees=30, twistLimitDegrees=6}
 {bone1Name=dyn_belt_r_02, bone2Name=dyn_belt_r_01, swing1LimitDegrees=10, swing2LimitDegrees=30, twistLimitDegrees=6}
 {bone1Name=dyn_belt_r_01, bone2Name=dyn_belt_r_master, swing1LimitDegrees=3, swing2LimitDegrees=30, twistLimitDegrees=3}
 {
  bone1Name=dyn_belt_r_06
  bone2Name=pack
  swing1LimitDegrees=45
  swing1Motion=Free
  swing2LimitDegrees=45
  swing2Motion=Free
  twistLimitDegrees=45
  twistMotion=Free
 }
 {
  bone1Name=dyn_zipper_upper_01
  bone2Name=dyn_zipper_upper_master
  swing1LimitDegrees=15
  swing2LimitDegrees=45
  twistLimitDegrees=45
  twistMotion=Locked
 }
]
