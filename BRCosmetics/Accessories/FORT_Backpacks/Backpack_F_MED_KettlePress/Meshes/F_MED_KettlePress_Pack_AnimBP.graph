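# Animation Blueprint for the KettlePress backpack, adjusting pack placement to the wearer's pose and simulating its physics.
# Event graph derives smoothed pack offsets, melee weight and gravity override from spine bend, movement state and body type.
# AnimGraph offsets the pack bone, springs the handle with AnimDynamics, then caches that pose.
# Two RigidBody sims share the cache, blending to the alternate physics asset when crouching or DBNO.
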AnimBlueprint /BRCosmetics/Accessories/FORT_Backpacks/Backpack_F_MED_KettlePress/Meshes/F_MED_KettlePress_Pack_AnimBP
target_skeleton=&/F_MED_KettlePress_Pack_Skeleton
graphs [
 AnimGraph {nodes [
   n1 = RigidBody!(ComponentPose=Local To Component(LocalPose=Use cached pose 'BackpackBasePose'()), Alpha, OverrideWorldGravity:Vector="0,0,0", ExternalForce) -> (Pose) {
    bClampLinearTranslationLimitToRefPose=true
    bOverrideWorldGravity=true
    overridePhysicsAsset=&/F_MED_KettlePress_Pack_AltPhysics
    alphaBoolBlend {blendOption=Linear}
    componentLinearAccScale=<0.4 0.4 0.4>
    componentLinearVelScale=<0.4 0.4 0.4>
    simSpaceSettings {maxAngularAcceleration=500, maxAngularVelocity=3, maxLinearAcceleration=400, maxLinearVelocity=250, externalLinearDragV=<0.1 0.1>}
   }
   n2 = RigidBody(ComponentPose=Local To Component(LocalPose=Use cached pose 'BackpackBasePose'()), OverrideWorldGravity="0,0,0") {
    bClampLinearTranslationLimitToRefPose=true
    bOverrideWorldGravity=true
    overridePhysicsAsset=&/F_MED_KettlePress_Pack_Physics
    alphaBoolBlend {blendOption=Linear}
    componentLinearAccScale=<0.4 0.4 0.4>
    componentLinearVelScale=<0.4 0.4 0.4>
    simSpaceSettings {maxAngularAcceleration=500, maxAngularVelocity=3, maxLinearAcceleration=400, maxLinearVelocity=250, externalLinearDragV=<0.1 0.1>}
   }
   Output Pose(Result=Blend Poses by bool(
     BlendPose_0=Component To Local(ComponentPose=@n1)
     BlendPose_1=Component To Local(ComponentPose=@n2)
     bActiveValue=Get altPhysics() -> (altPhysics:bool)
    ) {blendType=Linear}) {layerGroup=None}
   n3 = "Transform (Modify) Bone - Bone: pack"!(ComponentPose=Local To Component(LocalPose=Input Pose()), Alpha, Translation="0,-0.5,0") -> (Pose) {translationMode=BMM_Additive, alphaBoolBlend {blendOption=Linear}, boneToModify {boneName=pack}, translation {y=-0.5}}
   n4 = "Transform (Modify) Bone - Bone: pack"!(ComponentPose=@n3, Alpha, Translation="0,-1.5,0", Rotation:Rotator="0,0,0", Scale:Vector) -> (Pose) {
    rotationMode=BMM_Additive
    rotationSpace=BCS_BoneSpace
    translationMode=BMM_Additive
    translationSpace=BCS_BoneSpace
    alphaBoolBlend {blendOption=Linear}
    alphaScaleBiasClamp {interpSpeedDecreasing=5, interpSpeedIncreasing=5}
    boneToModify {boneName=pack}
    translation {y=-1.5}
   }
   n5 = "Transform (Modify) Bone - Bone: pack"(ComponentPose=@n4, Translation="0,0,0", Rotation="0,0,0") {rotationMode=BMM_Additive, translationMode=BMM_Additive, alphaBoolBlend {blendOption=Linear}, boneToModify {boneName=pack}}
   n6 = Anim Dynamics - Bone: dyn_handle_01!(ComponentPose=@n5, Alpha=1, GravityOverride:Vector="0,0,150", ExternalForce="0,0,0") -> (Pose) {
    angularDampingOverride=0.9999
    angularSpringConstant=200
    bAngularSpring=true
    bGravityOverrideInSimSpace=true
    bLinearSpring=true
    bOverrideAngularDamping=true
    bOverrideLinearDamping=true
    bUseGravityOverride=true
    bUsePlanarLimit=false
    linearDampingOverride=0.999
    linearSpringConstant=50
    simulationSpace=RootRelative
    alphaBoolBlend {blendOption=Linear}
    boundBone {boneName=dyn_handle_01}
    componentAppliedLinearAccClamp=<6000 6000 6000>
    componentLinearAccScale=<0.3 0.5 0.3>
    componentLinearVelScale=<0.3 0.5 0.3>
    gravityOverride {z=150}
    physicsBodyDefinitions [{
      boundBone {boneName=dyn_handle_01}
      boxExtents=<10 10 10>
      constraintSetup {
       angularConstraintType=Cone
       angularLimitsMax=<1 1 1>
       angularLimitsMin=<-1 -1 -1>
       angularTarget=<1 0 0>
       linearAxesMax=<0.1 0.1 0.1>
       linearAxesMin=<-0.1 -0.1 -0.1>
      }
      localJointOffset=<0.1 0 0>
     }]
    relativeSpaceBone {boneName=pack}
   }
   Save cached pose 'BackpackBasePose'(Pose=Component To Local(ComponentPose=@n6)) {cacheName=BackpackBasePose}
  ]}
 EventGraph {nodes [
   n1 = Event BlueprintUpdateAnimation()
   > n2 = Sequence!(execute) -> (then_0, then_1, then_2:Exec, then_3:Exec, then_4:Exec, then_5:Exec)
   Set altPhysics(execute=@n2.then_1, altPhysics:bool=OR Boolean(A=Get bIsDBNO(), B=Get bIsCrouching())) -> (Output_Get:bool)
   n3 = Branch(execute=@n2.then_2, Condition=Get bIsDBNO())
   > Set packDBNOTransY(packDBNOTransY:double=-1.8) -> (Output_Get:double)
   Set packDBNOTransY(execute=@n3.else, packDBNOTransY=0)
   n4 = Branch(execute=@n2.then_3, Condition=Get isM_Med() -> (isM_Med:bool))
   > n5 = Branch(Condition=Get bIsSkyDiving())
   n6 = GetOwnerSkeletalMeshComponent()
   n7 = BreakRotator(InRot=BreakTransform(InTransform=GetDeltaTransformFromRefPose(self=@n6, BoneName=spine_05, BaseName=pelvis)).Rotation)
   n8 = MapRangeClamped(Value=@n7.Yaw, InRangeA=10, InRangeB=0, OutRangeB=20)
   > Set TargetRotX(TargetRotX:double=@n8) -> (Output_Get:double)
   n9 = Branch(execute=@n5.else, Condition=Get bIsCrouching())
   n10 = MapRangeClamped(Value=@n7.Yaw, InRangeA=-30, InRangeB=-81, OutRangeB=65)
   n11 = MapRangeClamped(Value=@n7.Yaw, InRangeA=8, InRangeB=-8, OutRangeB=45)
   n12 = MapRangeClamped(Value=@n7.Yaw, InRangeA=-8, InRangeB=-30, OutRangeB=15)
   > Set TargetRotX(TargetRotX=float + float(A=10, B=float + float!(A=@n8, B=@n11, C:double=@n12, D:double=@n10) -> (ReturnValue)))
   n13 = Vector Length(A=Get PawnVelocity())
   n14 = Branch(execute=@n9.else, Condition=AND Boolean(A="float > float"(A=@n13), B="float < float"(A=@n13, B=150)))
   n15 = Get bIsSkydiveFloating()
   n16 = MapRangeClamped(Value="To Float (Boolean)"(InBool=@n15), InRangeA=0, InRangeB=1, OutRangeB=-50)
   n17 = "To Float (Boolean)"(InBool=AND Boolean(A=NOT Boolean(A=@n15), B=Get bIsSkyDiving()))
   n18 = MapRangeClamped(Value=@n7.Yaw, InRangeA=-30, InRangeB=-80, OutRangeA=1, OutRangeB=0)
   n19 = Get bIsInFrontEnd()
   n20 = float + float!(
    A=float * float(A=@n11, B:bool=NOT Boolean(A=@n19)) {is=title K2Node_PromotableOperator}
    B=@n12
    C=@n10
    D=float * float(A=@n18, B:double=MapRangeClamped(Value=@n17, InRangeA=0, InRangeB=1, OutRangeB=-30))
    E:double=@n16
   ) -> (ReturnValue)
   n21 = float * float(A=@n20, B=MapRangeClamped(Value=@n13, InRangeA=150, InRangeB=500))
   > Set TargetRotX(TargetRotX=float + float!(A=@n21, B=28) -> (ReturnValue))
   n22 = AND Boolean!(A="float < float"(A=@n13, B=50), B="float > float"(A=@n13), C:bool="float < float"(A=@n7.Yaw, B=10), D:bool="float > float"(A=@n7.Yaw)) -> (ReturnValue)
   Set TargetRotX(execute=@n14.else, TargetRotX=float + float(A=@n21, B=float * float(A=5, B:bool=@n22)))
   Set TargetRotX(execute=@n4.else, TargetRotX=@n20)
   n23 = Branch(execute=@n2.then_5, Condition=Get bIsInFrontEnd())
   n24 = GetOwnerSkeletalMeshComponent()
   n25 = Set Head Down Rot(
    execute=@n23.else
    Head Down Rot:float=GetAlphaFromReferenceTransformDelta(SkeletalMeshComponent=@n24, BoneName=spine_01, BaseName=head, InputRange=(X=0,Y=-45), TwistAxis="1,0,0")
   ) -> (Output_Get:float)
   n26 = "To Float (Boolean)"(InBool=Get bIsPlayingMeleeAnim())
   n27 = float * float(A=@n1.DeltaTimeX, B:double=0.5)
   n28 = Get isMelee() -> (isMelee:double)
   > Set isMelee(isMelee:double=float * float(A=@n25, B=FInterpTo(Current=@n28, Target=@n26, DeltaTime=@n27, InterpSpeed=5))) -> (Output_Get:double)
   n29 = BreakRotator(InRot=BreakTransform(InTransform=GetDeltaTransformFromRefPose(self=@n24, BoneName=upperarm_l, BaseName=clavicle_l)).Rotation)
   > Set Backpack Pickaxe Rot(Backpack Pickaxe Rot:Rotator=MakeRotator(Roll=MapRangeClamped(Value=@n29.Yaw, InRangeA=0, InRangeB=90, OutRangeB=30))) -> (Output_Get:Rotator)
   n30 = float * float(A=@n1.DeltaTimeX, B=Get DeltaTimeModifier() -> (DeltaTimeModifier:double))
   n31 = Get TargetRotX() -> (TargetRotX:double)
   n32 = Get!() -> (BackpackRotate_Roll:float, BackpackRotate_Pitch:float, BackpackRotate_Yaw:float)
   Set BackpackRotate(execute=@n2.then_4, BackpackRotate:Rotator=MakeRotator(Roll=FInterpTo(Current=@n32, Target=@n31, DeltaTime=@n30, InterpSpeed=1))) -> (Output_Get:Rotator)
   n33 = float + float!(
    A=MapRangeClamped(Value=@n7.Yaw, InRangeA=-8, InRangeB=-30, OutRangeB=-3)
    B=Get packDBNOTransY() -> (packDBNOTransY:double)
    C:double=MapRangeClamped(Value=@n7.Yaw, InRangeA=-50, InRangeB=-75, OutRangeB=-6.5)
   ) -> (ReturnValue)
   n34 = Get!() -> (BackpackTranslation_X:double, BackpackTranslation_Y:double, BackpackTranslation_Z:double)
   > Set BackpackTranslation(BackpackTranslation:Vector=MakeVector(Y=FInterpTo(Current=@n34.BackpackTranslation_Y, Target=@n33, DeltaTime=@n30, InterpSpeed=5))) -> (Output_Get:Vector)
   n35 = Gravity Override State Function DEPRECATED(
    execute=@n2.then_0
    CharacterPartAnimInstance=Self-Reference()
    GameTimeInSeconds=GetGameTimeInSeconds()
    GravityOverrideParams=Get Gravity Override Params()
    DebugJointName=spine_05
   )
   > Set Gravity Override(Gravity Override=@n35)
   n36 = Event OnCharacterCustomizationLoaded()
   n37 = "Get (a copy)"(Array:Array of Custom Character Part Object References=@n36.CustomParts, Dimension 1=1) -> (Output:Custom Character Part Object Reference)
   n38 = Get GenderPermitted!(self:Custom Character Part Object Reference=@n37) -> (GenderPermitted:EFortCustomGender Enum=Invalid)
   n39 = Get BodyTypesPermitted!(self:Custom Character Part Object Reference=@n37) -> (BodyTypesPermitted:EFortCustomBodyType Enum=NONE)
   > Set isM_Med(isM_Med:bool=AND Boolean!(
     A="Equal (Enum)"(A:EFortCustomBodyType Enum=@n39, B:EFortCustomBodyType Enum=Medium)
     B="Equal (Enum)"(A:EFortCustomGender Enum=@n38, B:EFortCustomGender Enum=Male)
    ) -> (ReturnValue)) -> (Output_Get:bool)
   float * float(A=@n18, B=@n16)
  ]}
]
variables [
 Alpha Output {type=number}
 altPhysics {type=boolean}
 Backpack Pickaxe Rot {type=Rotator}
 BackpackRotate {type=Rotator}
 BackpackTranslation {type=Vector}
 DeltaTimeModifier {type=number, default=2}
 Gravity Override {type=Vector}
 Gravity Override Params {type=GravityOverrideParamsStruct, default {dBNO {gravityOverride=<-980 1700 0>}}}
 Head Down Rot {type=number}
 isM_Med {type=boolean}
 isMelee {type=number}
 PackDBNOTransY {type=number}
 TargetRotX {type=number}
]
